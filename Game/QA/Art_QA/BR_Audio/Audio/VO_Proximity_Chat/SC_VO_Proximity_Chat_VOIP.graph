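# Sound cue in a QA audio test folder that plays a random voice sample as stand-in proximity voice chat (VOIP) audio.
# A single Random node picks with equal weight among thirty VO_VOIP_Sample wave players and is excluded from branch culling.
# Playback uses the proximity chat attenuation asset and a boosted volume multiplier, and routes to the SFX sound class rather than a voice-named one.
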SoundCue /Game/QA/Art_QA/BR_Audio/Audio/VO_Proximity_Chat/SC_VO_Proximity_Chat_VOIP
root=Random_0
nodes [
 {
  id=Random_0
  children [
   WavePlayer_0
   WavePlayer_1
   WavePlayer_2
   WavePlayer_3
   WavePlayer_4
   WavePlayer_5
   WavePlayer_6
   WavePlayer_7
   WavePlayer_8
   WavePlayer_9
   WavePlayer_10
   WavePlayer_11
   WavePlayer_12
   WavePlayer_13
   WavePlayer_14
   WavePlayer_15
   WavePlayer_16
   WavePlayer_17
   WavePlayer_18
   WavePlayer_19
   WavePlayer_20
   WavePlayer_21
   WavePlayer_22
   WavePlayer_23
   WavePlayer_24
   WavePlayer_25
   WavePlayer_26
   WavePlayer_27
   WavePlayer_28
   WavePlayer_29
  ]
  properties {bShouldExcludeFromBranchCulling=true, weights [1,1,1,1,1,1,1,1,1,1,1,1,1,1,1,1,1,1,1,1,1,1,1,1,1,1,1,1,1,1]}
 }
 {id=WavePlayer_0, properties {soundWaveAssetPtr=&/VO_VOIP_Sample_01}}
 {id=WavePlayer_1, properties {soundWaveAssetPtr=&/VO_VOIP_Sample_02}}
 {id=WavePlayer_10, properties {soundWaveAssetPtr=&/VO_VOIP_Sample_11}}
 {id=WavePlayer_11, properties {soundWaveAssetPtr=&/VO_VOIP_Sample_12}}
 {id=WavePlayer_12, properties {soundWaveAssetPtr=&/VO_VOIP_Sample_13}}
 {id=WavePlayer_13, properties {soundWaveAssetPtr=&/VO_VOIP_Sample_14}}
 {id=WavePlayer_14, properties {soundWaveAssetPtr=&/VO_VOIP_Sample_15}}
 {id=WavePlayer_15, properties {soundWaveAssetPtr=&/VO_VOIP_Sample_16}}
 {id=WavePlayer_16, properties {soundWaveAssetPtr=&/VO_VOIP_Sample_17}}
 {id=WavePlayer_17, properties {soundWaveAssetPtr=&/VO_VOIP_Sample_18}}
 {id=WavePlayer_18, properties {soundWaveAssetPtr=&/VO_VOIP_Sample_19}}
 {id=WavePlayer_19, properties {soundWaveAssetPtr=&/VO_VOIP_Sample_20}}
 {id=WavePlayer_2, properties {soundWaveAssetPtr=&/VO_VOIP_Sample_03}}
 {id=WavePlayer_20, properties {soundWaveAssetPtr=&/VO_VOIP_Sample_21}}
 {id=WavePlayer_21, properties {soundWaveAssetPtr=&/VO_VOIP_Sample_22}}
 {id=WavePlayer_22, properties {soundWaveAssetPtr=&/VO_VOIP_Sample_23}}
 {id=WavePlayer_23, properties {soundWaveAssetPtr=&/VO_VOIP_Sample_24}}
 {id=WavePlayer_24, properties {soundWaveAssetPtr=&/VO_VOIP_Sample_25}}
 {id=WavePlayer_25, properties {soundWaveAssetPtr=&/VO_VOIP_Sample_26}}
 {id=WavePlayer_26, properties {soundWaveAssetPtr=&/VO_VOIP_Sample_27}}
 {id=WavePlayer_27, properties {soundWaveAssetPtr=&/VO_VOIP_Sample_28}}
 {id=WavePlayer_28, properties {soundWaveAssetPtr=&/VO_VOIP_Sample_29}}
 {id=WavePlayer_29, properties {soundWaveAssetPtr=&/VO_VOIP_Sample_30}}
 {id=WavePlayer_3, properties {soundWaveAssetPtr=&/VO_VOIP_Sample_04}}
 {id=WavePlayer_4, properties {soundWaveAssetPtr=&/VO_VOIP_Sample_05}}
 {id=WavePlayer_5, properties {soundWaveAssetPtr=&/VO_VOIP_Sample_06}}
 {id=WavePlayer_6, properties {soundWaveAssetPtr=&/VO_VOIP_Sample_07}}
 {id=WavePlayer_7, properties {soundWaveAssetPtr=&/VO_VOIP_Sample_08}}
 {id=WavePlayer_8, properties {soundWaveAssetPtr=&/VO_VOIP_Sample_09}}
 {id=WavePlayer_9, properties {soundWaveAssetPtr=&/VO_VOIP_Sample_10}}
]
properties {
 attenuationSettings=&/ATT_VO_Proximity_Chat
 maxDistance=16000
 soundClassObject=&SFX
 volumeMultiplier=2
}
